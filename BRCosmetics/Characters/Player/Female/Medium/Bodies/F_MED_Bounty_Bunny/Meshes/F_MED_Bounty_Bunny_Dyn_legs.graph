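# Physics asset for the dynamic leg accessories of the Bounty Bunny female medium character body.
# Capsules cover right boot strings and left-leg grass pieces, bow and ribbon, each constrained to its side's calf_twist_02 bone.
# Swing limits are widest on the strings, moderate on grass and very tight on bow and ribbon, some with locked twist.
# Small pelvis and calf capsules act as anchors, with only the right calf linked to pelvis, by a free-rotation constraint.
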PhysicsAsset /BRCosmetics/Characters/Player/Female/Medium/Bodies/F_MED_Bounty_Bunny/Meshes/F_MED_Bounty_Bunny_Dyn_legs
bodies {
 calf_twist_02_l [{shapeType=Capsule, length=0.5, radius=0.5, rotation=<rot 90 0 0>}]
 calf_twist_02_r [{shapeType=Capsule, length=0.5, radius=0.5, rotation=<rot 90 0 0>}]
 dyn_boot_string_in_r [{shapeType=Capsule, length=3, radius=2, center=<1.069 -0.0001 0>, rotation=<rot 90 0 0>}]
 dyn_boot_string_out_r [{shapeType=Capsule, length=3, radius=2, center=<1.5858 0 0>, rotation=<rot 90 0 0>}]
 dyn_grass_l_small_02 [{shapeType=Capsule, length=5.5711, radius=0.9245, center=<-0.3276 -2.6394 -1.2731>, rotation=<rot -13.1246 -0.6619 55.8491>}]
 dyn_grass_mid_large_02 [{shapeType=Capsule, length=5.5558, radius=1.4817, center=<-0.5027 -1.1494 -5.5889>, rotation=<rot 4.1291 -90 1.6178>}]
 dyn_grass_mid_small_02 [{shapeType=Capsule, length=3.1686, radius=1.4, center=<1.5988 -0.8648 -3.0796>, rotation=<rot -2.1893 -90 -18.7193>}]
 dyn_grass_r_large_02 [{shapeType=Capsule, length=5.7812, radius=1.4357, center=<0.7549 -0.9872 -4.6547>, rotation=<rot 11.4932 -90 -8.344>}]
 dyn_leg_bow [{shapeType=Capsule, length=2.4426, radius=1.7332, center=<3.3337 -0.0978 0.4691>, rotation=<rot -15.5882 -89.9999 94.2311>}]
 dyn_leg_ribbon [{shapeType=Capsule, length=4.4184, radius=1.7687, center=<1.9169 0.1518 -0.9841>, rotation=<rot 62.6613 -20.1284 -22.4194>}]
 pelvis [{shapeType=Capsule, length=3, radius=2, rotation=<rot 90 0 0>}]
}
constraints [
 {bone1Name=dyn_boot_string_in_r, bone2Name=calf_twist_02_r, swing1LimitDegrees=30, swing2LimitDegrees=45, twistLimitDegrees=2}
 {bone1Name=dyn_boot_string_out_r, bone2Name=calf_twist_02_r, swing1LimitDegrees=30, swing2LimitDegrees=45, twistLimitDegrees=2}
 {
  bone1Name=calf_twist_02_r
  bone2Name=pelvis
  swing1LimitDegrees=45
  swing1Motion=Free
  swing2LimitDegrees=45
  swing2Motion=Free
  twistLimitDegrees=45
  twistMotion=Free
 }
 {
  bone1Name=dyn_grass_l_small_02
  bone2Name=calf_twist_02_l
  swing1LimitDegrees=6
  swing2LimitDegrees=8
  twistLimitDegrees=0
  twistMotion=Locked
 }
 {bone1Name=dyn_grass_mid_large_02, bone2Name=calf_twist_02_l, swing1LimitDegrees=12, swing2LimitDegrees=12, twistLimitDegrees=4}
 {bone1Name=dyn_grass_mid_small_02, bone2Name=calf_twist_02_l, swing1LimitDegrees=10, swing2LimitDegrees=10, twistLimitDegrees=2}
 {bone1Name=dyn_grass_r_large_02, bone2Name=calf_twist_02_l, swing1LimitDegrees=10, swing2LimitDegrees=10, twistLimitDegrees=4}
 {bone1Name=dyn_leg_ribbon, bone2Name=calf_twist_02_l, swing1LimitDegrees=2, swing2LimitDegrees=2, twistLimitDegrees=1, twistMotion=Locked}
 {bone1Name=dyn_leg_bow, bone2Name=calf_twist_02_l, swing1LimitDegrees=0.4, swing2LimitDegrees=0.4, twistLimitDegrees=45, twistMotion=Locked}
]
