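# Physics asset for the Bounty Bunny female medium body, simulating two dangling grenades and their pins as secondary motion.
# Each grenade chains from the pelvis capsule through a tiny master body to a grenade capsule, then to a box pin.
# Both master-to-pelvis joints and the first grenade's joint rotate freely, while the second grenade has tight limits.
# Pins swing widely on one axis only, like hinges.
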PhysicsAsset /BRCosmetics/Characters/Player/Female/Medium/Bodies/F_MED_Bounty_Bunny/Meshes/F_MED_Bounty_Bunny_Physics
bodies {
 dyn_grenade_01 [{shapeType=Capsule, length=4.0391, radius=3.355, center=<4.5589 0.36 1.2596>, rotation=<rot 0 0 15.9562>}]
 dyn_grenade_01_master [{shapeType=Box, extentX=1.01, extentY=1.01, extentZ=1.01, rotation=<rot 0 -90 0>}]
 dyn_grenade_02 [{shapeType=Capsule, length=5.7796, radius=3.1747, center=<4.0745 -0.2782 -2.3161>, rotation=<rot 20.2999 -3.5359 0.4174>}]
 dyn_grenade_02_master [{shapeType=Capsule, length=0.505, radius=0.505, rotation=<rot 0 -90 0>}]
 dyn_grenade_pin_01 [{shapeType=Box, extentX=4.3411, extentY=1.7435, extentZ=4.347, center=<0.6823 -0.1896 -1.7416>, rotation=<rot 6.1423 -90 -22.1599>}]
 dyn_grenade_pin_02 [{shapeType=Box, extentX=4.3288, extentY=1.6882, extentZ=4.347, center=<0.6532 0.1612 -2.346>, rotation=<rot -5.4418 -95.3572 -14.3207>}]
 pelvis [{shapeType=Capsule, length=6.0736, radius=12.5873, center=<-2.7334 -1.9192 -1.8299>, rotation=<rot -3.3559 -89.9985 79.4508>}]
}
constraints [
 {bone1Name=dyn_grenade_pin_02, bone2Name=dyn_grenade_02, swing1LimitDegrees=5, swing2LimitDegrees=75, twistLimitDegrees=5}
 {bone1Name=dyn_grenade_02, bone2Name=dyn_grenade_02_master, swing1LimitDegrees=20, swing2LimitDegrees=15, twistLimitDegrees=1}
 {
  bone1Name=dyn_grenade_02_master
  bone2Name=pelvis
  swing1LimitDegrees=45
  swing1Motion=Free
  swing2LimitDegrees=45
  swing2Motion=Free
  twistLimitDegrees=45
  twistMotion=Free
 }
 {
  bone1Name=dyn_grenade_01_master
  bone2Name=pelvis
  swing1LimitDegrees=45
  swing1Motion=Free
  swing2LimitDegrees=45
  swing2Motion=Free
  twistLimitDegrees=45
  twistMotion=Free
 }
 {
  bone1Name=dyn_grenade_01
  bone2Name=dyn_grenade_01_master
  swing1LimitDegrees=45
  swing1Motion=Free
  swing2LimitDegrees=45
  swing2Motion=Free
  twistLimitDegrees=45
  twistMotion=Free
 }
 {bone1Name=dyn_grenade_pin_01, bone2Name=dyn_grenade_01, swing1LimitDegrees=5, swing2LimitDegrees=75, twistLimitDegrees=5}
]
