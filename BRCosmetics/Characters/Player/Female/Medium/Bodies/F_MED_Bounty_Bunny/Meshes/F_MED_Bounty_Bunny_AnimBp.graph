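# Post-process style AnimBlueprint for the F_MED_Bounty_Bunny body, adding secondary motion and a crouch adjustment for its grenades.
# It copies the attached parent mesh's pose, then chains a thigh-based RigidBody leg simulation, two spring-driven AnimDynamics grenade bones and a pelvis-based RigidBody simulation.
# Crouch events toggle an interpolated alpha that additively offsets both grenade master bones.
# Two grenade Constraint nodes sit unconnected to the output.
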AnimBlueprint /BRCosmetics/Characters/Player/Female/Medium/Bodies/F_MED_Bounty_Bunny/Meshes/F_MED_Bounty_Bunny_AnimBp
target_skeleton=&/F_MED_Bounty_Bunny_Skeleton
graphs [
 AnimGraph {nodes [
   n1 = "Constraint - dyn_grenade_01 (P)"!(ComponentPose, Alpha=1, ConstraintWeights_0=1, ConstraintWeights_1:float=0) -> (Pose) {
    alphaBoolBlend {blendOption=Linear}
    boneToModify {boneName=dyn_grenade_01}
    constraintSetup [{targetBone {boneName=dyn_grenade_pin_01}}]
    constraintWeights [1,0]
   }
   "Constraint - dyn_grenade_02 (P)"!(ComponentPose=@n1, Alpha=1, ConstraintWeights_0=1, ConstraintWeights_1:float=0) -> (Pose) {
    alphaBoolBlend {blendOption=Linear}
    boneToModify {boneName=dyn_grenade_02}
    constraintSetup [{targetBone {boneName=dyn_grenade_02}}]
    constraintWeights [1,0]
   }
   n2 = RigidBody(ComponentPose=Local To Component(LocalPose=Copy Pose From Mesh() {bUseAttachedParent=true})) {
    bClampLinearTranslationLimitToRefPose=true
    bOverrideWorldGravity=true
    overridePhysicsAsset=&/F_MED_Bounty_Bunny_Dyn_legs
    simulationSpace=BaseBoneSpace
    alphaBoolBlend {blendOption=Linear}
    baseBoneRef {boneName=thigh_l}
    componentAppliedLinearAccClamp=<2000 2000 2000>
    componentLinearAccScale=<0.5 0.5 0.5>
    componentLinearVelScale=<0.5 0.5 0.5>
    overrideWorldGravity {z=-980}
   }
   n3 = Anim Dynamics - Bone: dyn_grenade_01(ComponentPose=@n2, Alpha=1, ExternalForce="0,0,0") {
    angularDampingOverride=0.9999
    angularSpringConstant=900
    bAngularSpring=true
    bGravityOverrideInSimSpace=true
    bOverrideAngularDamping=true
    bOverrideLinearDamping=true
    simulationSpace=RootRelative
    alphaBoolBlend {blendOption=Linear}
    boundBone {boneName=dyn_grenade_01}
    componentAppliedLinearAccClamp=<400 400 400>
    componentLinearAccScale=<0.15 0.15 0.15>
    componentLinearVelScale=<0.15 0.15 0.15>
    physicsBodyDefinitions [{
      sphereCollisionRadius=0
      boundBone {boneName=dyn_grenade_01}
      boxExtents=<30 30 5>
      constraintSetup {angularTargetAxis=AxisZ, twistAxis=AxisZ, angularLimitsMax=<10 10 10>, angularLimitsMin=<-10 -10 -10>, angularTarget=<0 0 0>}
      localJointOffset=<-0.2 0 -5>
     }]
   }
   n4 = Anim Dynamics - Bone: dyn_grenade_02(ComponentPose=@n3, Alpha=1, ExternalForce="0,0,0") {
    angularDampingOverride=0.99
    angularSpringConstant=1000
    bAngularSpring=true
    bGravityOverrideInSimSpace=true
    bOverrideAngularDamping=true
    bOverrideLinearDamping=true
    simulationSpace=RootRelative
    alphaBoolBlend {blendOption=Linear}
    boundBone {boneName=dyn_grenade_02}
    componentAppliedLinearAccClamp=<400 400 400>
    componentLinearAccScale=<0.15 0.15 0.15>
    componentLinearVelScale=<0.15 0.15 0.15>
    physicsBodyDefinitions [{
      sphereCollisionRadius=0
      boundBone {boneName=dyn_grenade_02}
      boxExtents=<40 40 8>
      constraintSetup {angularTargetAxis=AxisZ, twistAxis=AxisZ, angularLimitsMax=<10 10 10>, angularLimitsMin=<-10 -10 -10>, angularTarget=<0 0 0>}
      localJointOffset=<-0.2 0 -5>
     }]
   }
   n5 = RigidBody(ComponentPose=@n4) {
    bClampLinearTranslationLimitToRefPose=true
    bOverrideWorldGravity=true
    overridePhysicsAsset=&/F_MED_Bounty_Bunny_Physics
    simulationSpace=BaseBoneSpace
    alphaBoolBlend {blendOption=Linear}
    baseBoneRef {boneName=pelvis}
    componentAppliedLinearAccClamp=<5000 5000 5000>
    componentLinearAccScale=<1 1 0.15>
    componentLinearVelScale=<1 1 0.15>
    overrideWorldGravity {z=-980}
   }
   n6 = Get crouching alpha() -> (crouching alpha:double)
   n7 = "Transform (Modify) Bone - Bone: dyn_grenade_01_master"(ComponentPose=@n5, Alpha=@n6, Translation="-9,-11,0", Rotation="0,0,0", Scale="1,1,1") {
    rotationMode=BMM_Additive
    rotationSpace=BCS_BoneSpace
    translationMode=BMM_Additive
    translationSpace=BCS_BoneSpace
    alphaBoolBlend {blendOption=Linear}
    alphaScaleBiasClamp {bInterpResult=true}
    boneToModify {boneName=dyn_grenade_01_master}
    translation=<-9 -11>
   }
   n8 = "Transform (Modify) Bone - Bone: dyn_grenade_02_master"(ComponentPose=@n7, Alpha=@n6, Translation="-7,-10,0", Rotation="0,0,0", Scale="1,1,1") {
    rotationMode=BMM_Additive
    rotationSpace=BCS_BoneSpace
    translationMode=BMM_Additive
    translationSpace=BCS_BoneSpace
    alphaBoolBlend {blendOption=Linear}
    alphaScaleBiasClamp {bInterpResult=true}
    boneToModify {boneName=dyn_grenade_02_master}
    translation=<-7 -10>
   }
   Output Pose(Result=Component To Local(ComponentPose=@n8)) {layerGroup=None}
  ]}
 EventGraph {nodes [
   Event OnBeginCrouching()
   > Set crouching alpha(crouching alpha:double=1) -> (Output_Get:double)
   Event OnEndCrouching()
   > Set crouching alpha(crouching alpha=0)
   Event BlueprintUpdateAnimation()
  ]}
]
variables [crouching alpha {type=number}]
