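# Physics asset for the Bounty Bunny female medium player cosmetic's helmet, defining simulation bodies for its left and right dynamic straps.
# Each strap is a chain of three small capsules, with the top link constrained to a larger head capsule and each lower link to the one above.
# All joints share the same angular limits, with a narrow first swing axis and wide second swing and twist, so bending favors one plane.
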PhysicsAsset /BRCosmetics/Characters/Player/Female/Medium/Bodies/F_MED_Bounty_Bunny/Meshes/Parts/F_MED_Bounty_Bunny_Helmet_Phys
bodies {
 dyn_strap_01_l [{shapeType=Capsule, length=3.005, radius=1.005}]
 dyn_strap_01_r [{shapeType=Capsule, length=4.25, radius=1}]
 dyn_strap_02_l [{shapeType=Capsule, length=2.005, radius=1.255, center=<0 0 0.5153>}]
 dyn_strap_02_r [{shapeType=Capsule, length=3.505, radius=1.005, center=<0 0 -0.6089>, rotation=<rot 0 -90 0>}]
 dyn_strap_03_l [{shapeType=Capsule, length=3.255, radius=1.255, center=<0 0 -0.8503>}]
 dyn_strap_03_r [{shapeType=Capsule, length=3.627, radius=0.8795}]
 head [{shapeType=Capsule, length=9.5652, radius=7.6651, center=<0 0 0.2522>, rotation=<rot 90 0 0>}]
}
constraints [
 {bone1Name=dyn_strap_01_r, bone2Name=head, swing1LimitDegrees=7, swing2LimitDegrees=70, twistLimitDegrees=70}
 {bone1Name=dyn_strap_02_r, bone2Name=dyn_strap_01_r, swing1LimitDegrees=7, swing2LimitDegrees=70, twistLimitDegrees=70}
 {bone1Name=dyn_strap_03_r, bone2Name=dyn_strap_02_r, swing1LimitDegrees=7, swing2LimitDegrees=70, twistLimitDegrees=70}
 {bone1Name=dyn_strap_01_l, bone2Name=head, swing1LimitDegrees=7, swing2LimitDegrees=70, twistLimitDegrees=70}
 {bone1Name=dyn_strap_02_l, bone2Name=dyn_strap_01_l, swing1LimitDegrees=7, swing2LimitDegrees=70, twistLimitDegrees=70}
 {bone1Name=dyn_strap_03_l, bone2Name=dyn_strap_02_l, swing1LimitDegrees=7, swing2LimitDegrees=70, twistLimitDegrees=70}
]
